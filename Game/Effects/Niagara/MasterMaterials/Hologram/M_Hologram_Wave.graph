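# Translucent, unlit, two-sided hologram wave master material for Niagara sprite and mesh particles, with dynamic parameters controlling sweep, dots, displacement and refraction.
# Noise-distorted radial sweeps and procedural dot grids form a pattern that drives tinted emissive and ring-masked opacity.
# Use Discolor blends between an animated curve-atlas gradient and that pattern, optionally broken by scrolling bands.
# World position offset blends band and radial-sweep displacement; Forward and Mobile bypass refraction.
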Material /Game/Effects/Niagara/MasterMaterials/Hologram/M_Hologram_Wave
graph {
 properties {
  bAutomaticallySetUsageInEditor=false
  bUsedWithNiagaraMeshParticles=true
  bUsedWithNiagaraSprites=true
  blendMode=BLEND_Translucent
  refractionMethod=RM_IndexOfRefraction
  shadingModel=MSM_Unlit
  twoSided=true
  usedShadingModels=Unlit
 }
 nodes [
  n1 = Transform(None=Constant3Vector() {constant=<rgba 0 0 1 0>}) {transformSourceType=TRANSFORMSOURCE_Local}
  n2 = MaterialFunctionCall() {materialFunction=&VectorToRadialValue}
  n3 = Panner(Coordinate=Multiply(A=Add(A=@n2.Radial Coordinates), B=ScalarParameter() {defaultValue=1, parameterName=Texture_Tile_Master})) {speedY=-0.4}
  n4 = TextureCoordinate() {uTiling=0.5, vTiling=0.5}
  n5 = Sine(None=ComponentMask(None=@n4) {r=true})
  n6 = Sine(None=ComponentMask(None=@n4) {g=true})
  n7 = Power(Base=Multiply(A=Multiply(A=@n5, B=@n6)) {constB=3})
  n8 = OneMinus(None=SphereMask(A=TextureCoordinate(), B=Constant2Vector() {g=0.5, r=0.5}, Radius=Constant() {r=0.7}, Hardness=Constant() {r=0.3}))
  n9 = Multiply(
   A=LinearInterpolate(Alpha=TextureSample(UVs=@n3) {texture=/Game/Effects/Niagara/MasterMaterials/Hologram/Texture/T_Emote_AnglePatch_Sweep}.R) {constB=1.5}
   B=Saturate(None=Multiply(A=@n7, B=@n8))
  )
  n10 = ScalarParameter() {defaultValue=3.5, parameterName=WPO_Intensity}
  n11 = DynamicParameter() {paramNames [Sweep_Light, SweepSpeed_Y, WPO_Intensity, Refraction]}
  LinearInterpolate(
   A=NamedRerouteUsage()
   B=Multiply(A=Multiply(A=@n1, B=@n9), B=Multiply(A=@n10, B=@n11.WPO_Intensity)) {constB=10}
   Alpha=ScalarParameter() {defaultValue=1, group=Discolor, parameterName=Use WPOcolor}
  )
  n12 = ScalarParameter() {defaultValue=-0.4, parameterName=SweepSpeed_Y}
  n13 = DynamicParameter() {paramNames [Sweep_Light, SweepSpeed_Y, WPO_Intensity, Refraction]}
  n14 = Panner(
   Coordinate=MaterialFunctionCall() {materialFunction=&VectorToRadialValue}.Radial Coordinates
   Speed=AppendVector(A=Constant(), B=Multiply(A=@n12, B=@n13.SweepSpeed_Y))
  ) {speedY=-0.4}
  n15 = Constant2Vector() {g=1, r=1}
  n16 = MaterialFunctionCall() {materialFunction=&VectorToRadialValue}
  n17 = VectorParameter() {parameterName=noise speed, defaultValue=<rgba 1 1 0 1>}
  n18 = TextureSample(UVs=Multiply(A=Panner(Coordinate=@n16.Radial Coordinates) {speedX=0.1, speedY=-0.2}, B=ComponentMask(None=@n17.RGB) {g=true, r=true})) {samplerType=SAMPLERTYPE_LinearGrayscale, texture=/Game/Effects/Fort_Effects/Emotes/UltraLight/Texture/T_Cloud_Greyscale}
  n19 = ScalarParameter() {defaultValue=0.15, parameterName=Noise_Intensity}
  n20 = TextureSample(UVs=Add(A=Multiply(A=@n14, B=@n15), B=Multiply(A=@n18.R, B=@n19) {constB=0.2})) {texture=&T_Emote_AnglePatch_Sweep}
  n21 = LinearInterpolate(Alpha=TextureSample(UVs=Panner() {speedX=0.1, speedY=0.3}) {samplerType=SAMPLERTYPE_Grayscale, texture=&LowResBlurredNoise}.R) {constA=0.08}
  n22 = Constant() {r=1.5}
  n23 = DynamicParameter() {paramNames [Sweep_Light, SweepSpeed_Y, WPO_Intensity, Refraction]}
  n24 = Multiply(A=Add(A=TextureCoordinate(), B=Multiply(A=@n18.R) {constB=0.1})) {constB=100}
  n25 = DynamicParameter() {parameterIndex=1, defaultValue=<rgba 180 2.5 1 1>, paramNames [PointUV Tiling, NoisePointUV Tiling, Param3, Param4]}
  n26 = Multiply(A=TextureCoordinate(), B=@n25.PointUV Tiling) {constB=100}
  n27 = Constant() {r=2}
  n28 = Add(
   A=SphereMask(A=Fmod(A=@n24, B=@n25.NoisePointUV Tiling), B=Constant2Vector() {g=1, r=1}, Radius=Constant() {r=0.5}, Hardness=Constant() {r=0.8})
   B=SphereMask(
    A=Fmod(A=@n26, B=@n27)
    B=Constant2Vector() {g=1, r=1}
    Radius=ScalarParameter() {defaultValue=0.5, parameterName=Dot Radius}
    Hardness=Constant() {r=0.8}
   )
  )
  n29 = MaterialFunctionCall() {materialFunction=&VectorToRadialValue}
  n30 = LinearInterpolate(Alpha=TextureSample(UVs=Panner(Coordinate=@n29.Radial Coordinates) {speedY=-0.5}) {texture=/Game/Effects/Fort_Effects/Textures/Power_Source/T_Power_Source_Sphere_Texture}.R) {constA=0.12}
  n31 = Add(
   A=Multiply(A=Multiply(A=@n20.R, B=@n21), B=Multiply(A=@n22, B=@n23.Sweep_Light)) {constB=1.5}
   B=Multiply(A=Multiply(A=@n28, B=@n30)) {constB=1.5}
  )
  n32 = ParticleColor()
  n33 = Multiply(A=@n31, B=@n32.RGB)
  n34 = Multiply(A=TextureSample() {texture=&T_Tile_RainbowColor}.RGB, B=@n33)
  n35 = Constant() {r=-1.5}
  n36 = Add(
   A=Multiply(A=Desaturation(None=@n34, Fraction=@n35))
   B=Multiply(A=VectorParameter() {parameterName=color, defaultValue=<rgba 0.4538 3 0 1>}.RGB, B=@n33)
  )
  n37 = TextureCoordinate() {vTiling=5}
  n38 = Panner(Coordinate=@n37) {speedX=-2}
  n39 = Ceil(None=LinearInterpolate(Alpha=TextureSample(UVs=@n38) {samplerType=SAMPLERTYPE_Normal, texture=&T_Weap_Distort_N_02}.R) {constA=-0.16, constB=1.5})
  n40 = LinearInterpolate(Alpha=TextureSample(UVs=Panner(Coordinate=@n37) {speedX=2}) {samplerType=SAMPLERTYPE_Normal, texture=&T_Weap_Distort_N_02}.R) {constA=-0.2, constB=1.5}
  n41 = Subtract(A=Max(A=@n39, B=Ceil(None=@n40))) {constB=0.5}
  n42 = DynamicParameter() {parameterIndex=3, defaultValue=<rgba 1 -2 1 1>, paramNames [Use Discolor, Color WPO X, BaseWave_change, Param4]}
  n43 = LinearInterpolate(A=Multiply(A=@n36, B=@n41), B=@n36, Alpha=@n42.BaseWave_change)
  n44 = ScalarParameter() {defaultValue=1.2, parameterName=All light}
  n45 = TextureSample(UVs=Multiply(A=TextureCoordinate() {uTiling=0.5}) {constB=2}) {samplerType=SAMPLERTYPE_LinearColor, texture=/Game/TheOutpost/Textures/AnimatedOutpost_Base046_rgb_XVector_a_NormalizedHierPos_UV_2}
  n46 = Time()
  n47 = Sine(None=Add(A=Power(Base=@n45.B), B=Multiply(A=@n46) {constB=3}))
  n48 = CurveAtlasRowParameter(CurveTime=@n47) {
   atlas=/Game/Athena/LinearColorCurves/VFX_CurveAtlas
   curve=/Game/Accessories/FORT_Backpacks/Backpack_F_MED_Quarrel/FX/Curve_Quarrel_Auroras_01
   defaultValue=28
  }
  n49 = Constant() {r=-5}
  n50 = Multiply(A=Add(A=@n47, B=Desaturation(None=@n48, Fraction=@n49)), B=@n32.RGB)
  n51 = ScalarParameter() {parameterName=Des}
  MaterialFunctionCall(Emissive=LinearInterpolate(A=Desaturation(None=@n50, Fraction=@n51), B=Multiply(A=@n43, B=@n44), Alpha=@n42.Use Discolor)) {materialFunction=&MF_FX_EmissiveAdjustments}
  n52 = TextureCoordinate()
  n53 = Constant2Vector() {g=0.5, r=0.5}
  n54 = ScalarParameter() {defaultValue=0.3, group=Masks, parameterName=CenterMask_Radius, sortPriority=3}
  n55 = ScalarParameter() {defaultValue=0.6, group=Masks, parameterName=CenterMask_Hardness, sortPriority=4}
  n56 = Multiply(
   A=SphereMask(
    A=TextureCoordinate()
    B=Constant2Vector() {g=0.5, r=0.5}
    Radius=ScalarParameter() {defaultValue=0.5, group=Masks, parameterName=OuterMask_Radius, sortPriority=1}
    Hardness=ScalarParameter() {defaultValue=-1, group=Masks, parameterName=OuterMask_Hardness, sortPriority=2}
   )
   B=OneMinus(None=SphereMask(A=@n52, B=@n53, Radius=@n54, Hardness=@n55))
  )
  Multiply(A=Multiply(A=@n32.A, B=@n31), B=@n56)
  n57 = Multiply(A=Transform(None=Constant3Vector() {constant=<rgba 1 0 0 0>}) {transformSourceType=TRANSFORMSOURCE_Local}, B=@n41)
  n58 = ScalarParameter() {defaultValue=1, parameterName=Color X WPO_Intensity}
  n59 = Multiply(A=Transform(None=Constant3Vector() {constant=<rgba 0 0 1 0>}) {transformSourceType=TRANSFORMSOURCE_Local}, B=@n39)
  n60 = ScalarParameter() {defaultValue=1, parameterName=Color Z WPO_Intensity}
  NamedRerouteDeclaration(None=Add(A=Multiply(A=@n57, B=@n58) {constB=10}, B=Multiply(A=@n59, B=@n60) {constB=10})) {name=WPO_Color, nodeColor=<rgba 0 0.4844 1 1>}
  n61 = Constant() {r=1}
  n62 = LinearInterpolate(
   A=LinearInterpolate(B=DynamicParameter() {paramNames [Sweep_Light, SweepSpeed_Y, WPO_Intensity, Refraction]}.Refraction, Alpha=@n56) {constA=1, constB=0}
   B=@n61
   Alpha=ScalarParameter() {defaultValue=1, parameterName=Use_Refraction}
  )
  ShadingPathSwitch(Default=@n62, Deferred=@n62, Forward=@n61, Mobile=@n61)
 ]
}
