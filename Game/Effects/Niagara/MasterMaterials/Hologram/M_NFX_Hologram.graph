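# Translucent unlit hologram master material for Niagara mesh particles, built on emote coin normal and AO textures.
# Emissive blends two colors by fake dynamic-parameter lighting (LightDirectionX feeds all three axes) and AO, then applies Fresnel, tinted scrolling screen-space scanlines and a UV mask.
# Opacity combines Fresnel, scanlines, height fade and particle alpha, with cheaper quality and forward-shading paths.
# World-position-offset math pushes vertices outward horizontally using flickering noise and a TurnOn-driven height pulse.
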Material /Game/Effects/Niagara/MasterMaterials/Hologram/M_NFX_Hologram
graph {
 properties {
  allowTranslucentCustomDepthWrites=true
  bEnableResponsiveAA=true
  bUsedWithNiagaraMeshParticles=true
  blendMode=BLEND_Translucent
  pixelDepthOffsetMode=PDOM_Legacy
  previewMesh=/BRCosmetics/Animation/Props/Coin/Meshes/Coin
  shadingModel=MSM_Unlit
  usedShadingModels=Unlit
 }
 nodes [
  n1 = DynamicParameter() {paramNames [LightDirectionX, LightDirectionY, LightDirectionZ, Param4]}
  n2 = Transform(None=MaterialFunctionCall(X=@n1.LightDirectionX, Y=@n1.LightDirectionX, Z=@n1.LightDirectionX) {materialFunction=&MakeFloat3}.Result)
  n3 = Transform(None=TextureSampleParameter2D() {parameterName=Normal, samplerType=SAMPLERTYPE_Normal, texture=/Game/Animation/Props/Coin/Textures/T_Emote_Prop_Coin_N}.RGB)
  n4 = Power(Base=Clamp(None=DotProduct(A=@n2, B=@n3))) {constExponent=0.6}
  n5 = Power(Base=OneMinus(None=TextureSample() {texture=/Game/Effects/Fort_Effects/Textures/T_Emote_Prop_Coin_AO_V2}.R)) {constExponent=3}
  n6 = ScalarParameter() {defaultValue=1, parameterName=DetailEmissiveEdge}
  n7 = LinearInterpolate(
   A=VectorParameter() {parameterName=TopColor, defaultValue=<rgba 0.1222 0.2469 0.7 1>}.RGB
   B=VectorParameter() {parameterName=BottomColor, defaultValue=<rgba 0.0516 0.19 1 1>}.RGB
   Alpha=Add(A=Abs(None=@n4), B=Multiply(A=@n5, B=@n6))
  )
  n8 = MaterialFunctionCall(Distance=ScalarParameter() {defaultValue=1500, parameterName=FadeDistance}) {materialFunction=&Near_Camera_Fade}
  n9 = LinearInterpolate(Alpha=@n8.Result) {constA=3.2}
  n10 = Constant() {r=3}
  n11 = Fresnel(ExponentIn=ShadingPathSwitch(Default=QualitySwitch(Default=@n9, Low=@n10), Forward=@n10), BaseReflectFractionIn=@n4) {exponent=3}
  n12 = VertexNormalWS()
  n13 = ComponentMask(None=@n12) {b=true}
  n14 = ComponentMask(None=CameraVectorWS()) {b=true}
  n15 = Multiply(
   A=Fresnel() {baseReflectFraction=0, exponent=3}
   B=Multiply(A=DotProduct(A=@n13, B=@n14), B=ScalarParameter() {defaultValue=0.5, parameterName=LineBend})
  ) {constB=0.2}
  n16 = Constant()
  n17 = ComponentMask(None=ScreenPosition().ViewportUV) {g=true}
  n18 = Divide(A=Time()) {constB=3}
  n19 = Multiply(
   A=Add(A=QualitySwitch(Default=@n15, Low=@n16, Medium=@n16), B=Add(A=@n17, B=@n18))
   B=ScalarParameter() {defaultValue=150, parameterName=ScanlineTile}
  )
  n20 = Power(Base=ConstantBiasScale(None=Sine(None=@n19)))
  n21 = Constant3Vector() {constant=<rgba 1 0.2611 0.0197 0>}
  n22 = Constant3Vector() {constant=<rgba 0.025 1 0.1346 0>}
  n23 = ComponentMask(None=Transform(None=@n12) {transformSourceType=TRANSFORMSOURCE_World, transformType=TRANSFORM_View}) {r=true}
  n24 = Multiply(A=Add(A=LinearInterpolate(A=@n21, B=@n22, Alpha=@n23)) {constB=0.3})
  n25 = Add(A=LinearInterpolate(B=@n24, Alpha=@n20) {constA=1}, B=@n24)
  n26 = Constant() {r=1}
  n27 = Multiply(A=Clamp(None=@n11) {minDefault=0.25}, B=ShadingPathSwitch(Default=QualitySwitch(Default=@n25, Low=@n26), Forward=@n26)) {constA=1}
  n28 = Constant() {r=0.5}
  n29 = SphereMask(A=TextureCoordinate() {coordinateIndex=1}, B=@n28, Radius=@n28) {hardnessPercent=0}
  n30 = Multiply(A=@n29) {constB=2}
  MaterialFunctionCall(Emissive=Multiply(A=Multiply(A=@n7, B=@n27), B=Saturate(None=@n30))) {materialFunction=&MF_FX_EmissiveAdjustments}
  n31 = ParticlePositionWS()
  n32 = Subtract(A=ComponentMask(None=WorldPosition().XYZ) {b=true}, B=Add(A=ComponentMask(None=@n31) {b=true}) {constB=-29})
  n33 = Add(A=@n11, B=ScalarParameter() {defaultValue=0.4, parameterName=ScanlineAdd})
  n34 = LinearInterpolate(A=@n20, Alpha=@n8.Result) {constA=0.4, constB=0.3}
  n35 = Multiply(A=Saturate(None=Divide(A=@n32) {constB=80}), B=Clamp(None=Add(A=@n33, B=@n34) {constB=0.2}))
  n36 = ScalarParameter() {defaultValue=1, parameterName=OverallOpacity}
  n37 = Multiply(A=@n29) {constB=2}
  Multiply(A=ParticleColor().A, B=Multiply(A=Multiply(A=@n35, B=@n36) {constB=5}, B=Saturate(None=@n37)))
  n38 = Normalize(VectorInput=Subtract(A=WorldPosition().XYZ, B=ObjectPositionWS()))
  n39 = Constant3Vector() {constant=<rgba 1 1 0 0>}
  n40 = ScalarParameter() {defaultValue=10, group=Noise, parameterName=WPOStrength}
  n41 = Multiply(A=ObjectRadius(), B=ScalarParameter() {defaultValue=0.025, group=Noise, parameterName=WPOStrengthFromObjectRadius}) {constB=0.05}
  n42 = VectorParameter() {group=Noise, parameterName=NoiseTileMult, defaultValue=<rgba 1 1 0.1 1>}
  n43 = Divide(A=ObjectRadius(), B=ScalarParameter() {defaultValue=0.1, group=Noise, parameterName=NoiseSize})
  n44 = WorldPosition()
  n45 = Multiply(A=@n43, B=Multiply(A=Time(), B=VectorParameter() {group=Noise, parameterName=NoisePanSpeed, defaultValue=<rgba -0.08 0.07 -0.2 0>}.RGB))
  n46 = Power(Base=MaterialFunctionCall(
    TextureObject=TextureObject() {samplerType=SAMPLERTYPE_LinearColor, texture=&T_NoisePack_LinearColor}
    TextureSize=Multiply(A=@n42.RGB, B=@n43)
    WorldPosition=Add(A=@n44.XYZ, B=@n45)
   ) {materialFunction=&WorldAlignedTexture}.XY Texture) {constExponent=4}
  n47 = ScalarParameter() {defaultValue=2, group=Noise, parameterName=NoiseMaskMult}
  n48 = ScalarParameter() {group=Noise, parameterName=NoiseAdd}
  n49 = ActorPositionWS()
  n50 = MaterialFunctionCall(Float2=Round(None=Divide(A=@n49) {constB=2000})) {materialFunction=&BreakOutFloat2Components}
  n51 = Add(A=Time(), B=Add(A=@n50.R, B=@n50.R))
  n52 = Multiply(A=ScalarParameter() {defaultValue=0.5, parameterName=FlickerSpeed}, B=@n51)
  n53 = AppendVector(A=AppendVector(A=Sine(None=@n52) {period=1.6}, B=Sine(None=@n52) {period=1.42}), B=Sine(None=@n52) {period=1.51})
  n54 = ScalarParameter() {defaultValue=0.5, group=Noise, parameterName=NoiseFrequency}
  n55 = Add(A=Sine(None=@n52) {period=2.1}, B=Multiply(B=@n54) {constA=0.25})
  n56 = Add(A=Sine(None=@n52) {period=0.72}, B=Multiply(B=@n54) {constA=0.75}) {constB=0.75}
  n57 = Ceil(None=Add(A=Multiply(B=@n54) {constA=0.9, constB=0.9}, B=Sine(None=@n52) {period=1.25}) {constB=0.9})
  n58 = Saturate(None=Multiply(A=Multiply(A=@n55, B=@n56), B=Multiply(A=@n56, B=@n57)))
  n59 = LinearInterpolate(Alpha=ConstantBiasScale(None=Sine(None=@n51) {period=0.075})) {constA=0.7}
  n60 = Multiply(A=Multiply(A=Ceil(None=@n53), B=Multiply(A=@n58, B=@n59)), B=ScalarParameter() {defaultValue=2, group=Noise, parameterName=NoiseMult})
  n61 = Saturate(None=Multiply(A=Multiply(A=@n46, B=@n47) {constB=20}, B=Add(A=@n48, B=@n60)))
  n62 = WorldPosition()
  n63 = ComponentMask(None=MaterialFunctionCall() {materialFunction=&ObjectPivotPoint}.Object Pivot Location) {b=true}
  n64 = Add(A=@n63, B=ScalarParameter() {defaultValue=300, parameterName=PulseHeight}) {constB=1600}
  n65 = ScalarParameter() {defaultValue=1, parameterName=TurnOn}
  n66 = ScalarParameter() {defaultValue=33, parameterName=PulseRadius}
  n67 = Multiply(A=@n66) {constB=0.1}
  n68 = Multiply(
   A=SphereMask(
    A=ComponentMask(None=@n62.XYZ) {b=true}
    B=LinearInterpolate(A=@n63, B=@n64, Alpha=@n65) {constA=-0.25, constB=1.2}
    Radius=LinearInterpolate(A=@n66, B=@n67, Alpha=@n65) {constA=500, constB=100}
   ) {hardnessPercent=0}
   B=OneMinus(None=@n65)
  )
  Multiply(A=Add(A=Desaturation(None=@n61), B=Multiply(A=@n68) {constB=2}), B=Multiply(A=Multiply(A=@n38, B=@n39), B=Add(A=@n40, B=@n41))) {constB=0}
  Power(Base=Multiply(A=TextureSample() {samplerType=SAMPLERTYPE_Masks, texture=/Game/Animation/Props/Coin/Textures/T_Emote_Prop_Coin_S}.B))
 ]
}
